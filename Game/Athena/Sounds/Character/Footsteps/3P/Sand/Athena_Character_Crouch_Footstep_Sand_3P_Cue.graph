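# Sound cue for third-person crouch footsteps on sand in Athena, using crouch attenuation, the enemy foley sound class and footstep concurrency sets.
# A quality-level node picks a random pool of eight sand walk samples for the first two levels, or four for the other two.
# It reuses first-person walk recordings, with a modulator raising pitch and varying volume, under a reduced overall volume multiplier.
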SoundCue /Game/Athena/Sounds/Character/Footsteps/3P/Sand/Athena_Character_Crouch_Footstep_Sand_3P_Cue
root=Modulator_0
nodes [
 {id=Modulator_0, children [QualityLevel_0], properties {pitchMax=1.5, pitchMin=1.3, volumeMax=1, volumeMin=0.7}}
 {id=QualityLevel_0, children [Random_2,Random_2,Random_3,Random_3]}
 {
  id=Random_2
  children [WavePlayer_10, WavePlayer_11, WavePlayer_12, WavePlayer_13, WavePlayer_14, WavePlayer_15, WavePlayer_16, WavePlayer_17]
  properties {weights [1,1,1,1,1,1,1,1]}
 }
 {
  id=Random_3
  children [WavePlayer_18, WavePlayer_19, WavePlayer_20, WavePlayer_21]
  properties {bShouldExcludeFromBranchCulling=true, weights [1,1,1,1]}
 }
 {id=WavePlayer_10, properties {soundWaveAssetPtr=/Game/Athena/Sounds/Character/Footsteps/1P/Sand/FS_Sand_Walk_08}}
 {id=WavePlayer_11, properties {soundWaveAssetPtr=/Game/Athena/Sounds/Character/Footsteps/1P/Sand/FS_Sand_Walk_07}}
 {id=WavePlayer_12, properties {soundWaveAssetPtr=/Game/Athena/Sounds/Character/Footsteps/1P/Sand/FS_Sand_Walk_06}}
 {id=WavePlayer_13, properties {soundWaveAssetPtr=/Game/Athena/Sounds/Character/Footsteps/1P/Sand/FS_Sand_Walk_05}}
 {id=WavePlayer_14, properties {soundWaveAssetPtr=/Game/Athena/Sounds/Character/Footsteps/1P/Sand/FS_Sand_Walk_04}}
 {id=WavePlayer_15, properties {soundWaveAssetPtr=/Game/Athena/Sounds/Character/Footsteps/1P/Sand/FS_Sand_Walk_03}}
 {id=WavePlayer_16, properties {soundWaveAssetPtr=/Game/Athena/Sounds/Character/Footsteps/1P/Sand/FS_Sand_Walk_02}}
 {id=WavePlayer_17, properties {soundWaveAssetPtr=/Game/Athena/Sounds/Character/Footsteps/1P/Sand/FS_Sand_Walk_01}}
 {id=WavePlayer_18, properties {soundWaveAssetPtr=&FS_Sand_Walk_04}}
 {id=WavePlayer_19, properties {soundWaveAssetPtr=&FS_Sand_Walk_03}}
 {id=WavePlayer_20, properties {soundWaveAssetPtr=&FS_Sand_Walk_02}}
 {id=WavePlayer_21, properties {soundWaveAssetPtr=&FS_Sand_Walk_01}}
]
properties {
 attenuationSettings=/Game/Sounds/Attenuation/Fort_Player_Footsteps_3P_Crouch_Athena_Att
 maxDistance=2048
 soundClassObject=&Fort_Player_Foley_3P_Enemy_Athena
 volumeMultiplier=0.3
 concurrencyOverrides {bLimitToOwner=true, maxCount=2, resolutionRule=StopOldest}
 concurrencySet [&Athena_Footstep_Local_Conc, &Athena_Footstep_Global_Conc]
}
