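# Third-person sand jump sound cue for characters, routed to the enemy foley sound class with player-jump attenuation.
# A quality-level switch picks randomly from four sand jump waves in its first two slots and two in the last two, all reusing first-person assets.
# A modulator varies volume and raises pitch on each play.
# Concurrency allows one instance per owner and blocks new plays while it sounds.
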SoundCue /Game/Athena/Sounds/Character/Footsteps/3P/Sand/Athena_Character_Jump_Sand_3P_Cue
root=Modulator_2
nodes [
 {id=Modulator_2, children [QualityLevel_0], properties {pitchMax=1.3, pitchMin=1.1, volumeMax=1, volumeMin=0.7}}
 {id=QualityLevel_0, children [Random_3,Random_3,Random_0,Random_0]}
 {id=Random_0, children [WavePlayer_10, WavePlayer_11], properties {bShouldExcludeFromBranchCulling=true, weights [1,1]}}
 {id=Random_3, children [WavePlayer_6,WavePlayer_7,WavePlayer_8,WavePlayer_9], properties {weights [1,1,1,1]}}
 {id=WavePlayer_10, properties {soundWaveAssetPtr=/Game/Athena/Sounds/Character/Footsteps/1P/Sand/FS_Sand_Jump_02}}
 {id=WavePlayer_11, properties {soundWaveAssetPtr=/Game/Athena/Sounds/Character/Footsteps/1P/Sand/FS_Sand_Jump_01}}
 {id=WavePlayer_6, properties {soundWaveAssetPtr=/Game/Athena/Sounds/Character/Footsteps/1P/Sand/FS_Sand_Jump_04}}
 {id=WavePlayer_7, properties {soundWaveAssetPtr=/Game/Athena/Sounds/Character/Footsteps/1P/Sand/FS_Sand_Jump_03}}
 {id=WavePlayer_8, properties {soundWaveAssetPtr=&FS_Sand_Jump_02}}
 {id=WavePlayer_9, properties {soundWaveAssetPtr=&FS_Sand_Jump_01}}
]
properties {
 attenuationSettings=/Game/Sounds/Attenuation/Fort_Player_Jump_3P_Athena_Att
 bOverrideConcurrency=true
 maxDistance=3584
 soundClassObject=&Fort_Player_Foley_3P_Enemy_Athena
 volumeMultiplier=0.5
 concurrencyOverrides {bLimitToOwner=true, maxCount=1, resolutionRule=PreventNew}
}
