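# Third-person hard-landing sound cue for characters on sand, routed to the enemy foley class with player-landing attenuation and global footstep concurrency.
# Each variant mixes a random soldier landing with a random sand landing reused from first-person footstep assets.
# A quality switch gives its first two levels four sand takes and its last two only two.
# A root modulator randomizes pitch and volume.
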SoundCue /Game/Athena/Sounds/Character/Footsteps/3P/Sand/Athena_Character_LandHard_Sand_3P_Cue
root=Modulator_4
nodes [
 {id=Mixer_0, children [Random_0,Random_5], properties {inputVolume [1,1]}}
 {id=Mixer_2, children [Random_3,Random_2], properties {inputVolume [1,1]}}
 {id=Modulator_4, children [QualityLevel_0], properties {pitchMax=1.1, pitchMin=0.9, volumeMax=1.3, volumeMin=0.7}}
 {id=QualityLevel_0, children [Mixer_2,Mixer_2,Mixer_0,Mixer_0]}
 {id=Random_0, children [WavePlayer_6,WavePlayer_7], properties {bShouldExcludeFromBranchCulling=true, weights [1,1]}}
 {id=Random_2, children [WavePlayer_2, WavePlayer_8, WavePlayer_9, WavePlayer_10], properties {weights [1,1,1,1]}}
 {id=Random_3, children [WavePlayer_19, WavePlayer_20], properties {bShouldExcludeFromBranchCulling=true, weights [1,1]}}
 {id=Random_5, children [WavePlayer_12, WavePlayer_11], properties {bShouldExcludeFromBranchCulling=true, weights [1,1]}}
 {id=WavePlayer_10, properties {soundWaveAssetPtr=/Game/Athena/Sounds/Character/Footsteps/1P/Sand/FS_Sand_Land_01}}
 {id=WavePlayer_11, properties {soundWaveAssetPtr=&FS_Sand_Land_01}}
 {id=WavePlayer_12, properties {soundWaveAssetPtr=/Game/Athena/Sounds/Character/Footsteps/1P/Sand/FS_Sand_Land_02}}
 {id=WavePlayer_19, properties {soundWaveAssetPtr=/Game/Sounds/Fort_Step_Sounds/Human/soldier_land_01}}
 {id=WavePlayer_2, properties {soundWaveAssetPtr=/Game/Athena/Sounds/Character/Footsteps/1P/Sand/FS_Sand_Land_04}}
 {id=WavePlayer_20, properties {soundWaveAssetPtr=/Game/Sounds/Fort_Step_Sounds/Human/soldier_land_02}}
 {id=WavePlayer_6, properties {soundWaveAssetPtr=&soldier_land_01}}
 {id=WavePlayer_7, properties {soundWaveAssetPtr=&soldier_land_02}}
 {id=WavePlayer_8, properties {soundWaveAssetPtr=/Game/Athena/Sounds/Character/Footsteps/1P/Sand/FS_Sand_Land_03}}
 {id=WavePlayer_9, properties {soundWaveAssetPtr=&FS_Sand_Land_02}}
]
properties {
 attenuationSettings=/Game/Sounds/Attenuation/Fort_Player_Land_3P_Athena_Att
 maxDistance=3584
 soundClassObject=&Fort_Player_Foley_3P_Enemy_Athena
 volumeMultiplier=1.4
 concurrencyOverrides {bLimitToOwner=true, maxCount=4, resolutionRule=StopOldest}
 concurrencySet [&Athena_Footstep_Global_Conc]
}
