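# Sound cue for third-person sprint footsteps on sand in Athena, routed to the enemy foley sound class.
# A Quality Level node sends its first two slots to eight random variations and its last two to four.
# It reuses first-person sand walk waves, with a Modulator raising pitch and randomizing pitch and volume.
# Sprint footstep attenuation spatializes it, with local and global footstep concurrency limiting overlapping steps.
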SoundCue /Game/Athena/Sounds/Character/Footsteps/3P/Sand/Athena_Character_Sprint_Footstep_Sand_3P_Cue
root=Modulator_2
nodes [
 {id=Modulator_2, children [QualityLevel_0], properties {pitchMax=1.5, pitchMin=1.3, volumeMax=1, volumeMin=0.7}}
 {id=QualityLevel_0, children [Random_2,Random_2,Random_4,Random_4]}
 {
  id=Random_2
  children [WavePlayer_3, WavePlayer_4, WavePlayer_5, WavePlayer_15, WavePlayer_16, WavePlayer_17, WavePlayer_24, WavePlayer_25]
  properties {weights [1,1,1,1,1,1,1,1]}
 }
 {
  id=Random_4
  children [WavePlayer_26, WavePlayer_27, WavePlayer_28, WavePlayer_29]
  properties {bShouldExcludeFromBranchCulling=true, weights [1,1,1,1]}
 }
 {id=WavePlayer_15, properties {soundWaveAssetPtr=/Game/Athena/Sounds/Character/Footsteps/1P/Sand/FS_Sand_Walk_05}}
 {id=WavePlayer_16, properties {soundWaveAssetPtr=/Game/Athena/Sounds/Character/Footsteps/1P/Sand/FS_Sand_Walk_04}}
 {id=WavePlayer_17, properties {soundWaveAssetPtr=/Game/Athena/Sounds/Character/Footsteps/1P/Sand/FS_Sand_Walk_03}}
 {id=WavePlayer_24, properties {soundWaveAssetPtr=/Game/Athena/Sounds/Character/Footsteps/1P/Sand/FS_Sand_Walk_02}}
 {id=WavePlayer_25, properties {soundWaveAssetPtr=/Game/Athena/Sounds/Character/Footsteps/1P/Sand/FS_Sand_Walk_01}}
 {id=WavePlayer_26, properties {soundWaveAssetPtr=&FS_Sand_Walk_04}}
 {id=WavePlayer_27, properties {soundWaveAssetPtr=&FS_Sand_Walk_03}}
 {id=WavePlayer_28, properties {soundWaveAssetPtr=&FS_Sand_Walk_02}}
 {id=WavePlayer_29, properties {soundWaveAssetPtr=&FS_Sand_Walk_01}}
 {id=WavePlayer_3, properties {soundWaveAssetPtr=/Game/Athena/Sounds/Character/Footsteps/1P/Sand/FS_Sand_Walk_08}}
 {id=WavePlayer_4, properties {soundWaveAssetPtr=/Game/Athena/Sounds/Character/Footsteps/1P/Sand/FS_Sand_Walk_07}}
 {id=WavePlayer_5, properties {soundWaveAssetPtr=/Game/Athena/Sounds/Character/Footsteps/1P/Sand/FS_Sand_Walk_06}}
]
properties {
 attenuationSettings=/Game/Sounds/Attenuation/Fort_Player_Footsteps_3P_Sprint_Athena_Att
 maxDistance=3584
 soundClassObject=&Fort_Player_Foley_3P_Enemy_Athena
 volumeMultiplier=0.7
 concurrencyOverrides {bLimitToOwner=true, maxCount=2, resolutionRule=StopOldest}
 concurrencySet [&Athena_Footstep_Local_Conc, &Athena_Footstep_Global_Conc]
}
